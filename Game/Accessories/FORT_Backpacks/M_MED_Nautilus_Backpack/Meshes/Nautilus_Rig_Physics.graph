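# Physics asset for the Nautilus backpack rig, a back accessory, giving it simulated secondary motion.
# It has two box bodies: a tiny one on backpack_base and a large offset box on dyn_backpack covering the pack.
# A single constraint links dyn_backpack to backpack_base with small swing limits and almost no twist, so the pack can sway slightly relative to its base.
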PhysicsAsset /Game/Accessories/FORT_Backpacks/M_MED_Nautilus_Backpack/Meshes/Nautilus_Rig_Physics
bodies {
 backpack_base [{shapeType=Box, extentX=1.01, extentY=1.01, extentZ=1.01}]
 dyn_backpack [{shapeType=Box, extentX=33.2449, extentY=15.1857, extentZ=41.2116, center=<0.6195 4.5474 -0.2314>}]
}
constraints [{bone1Name=dyn_backpack, bone2Name=backpack_base, swing1LimitDegrees=10, swing2LimitDegrees=15, twistLimitDegrees=1}]
