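# Animation Blueprint for the Nautilus male backpack cosmetic that animates four pistons procedurally.
# Each update, the EventGraph maps phase-offset sine waves of game time to piston01 through piston04, producing a staggered cycle.
# These values drive chained, interpolated Apply Additive nodes that layer matching pose asset poses onto the reference pose.
# A RigidBody physics branch with overridden gravity exists but is not connected to the output pose.
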AnimBlueprint /BRCosmetics/Accessories/FORT_Backpacks/M_MED_Nautilus_Backpack/Meshes/Nautilus_Rig_Backpack_Nautilus_Male_MESH_Skeleton_AnimBP
target_skeleton=/Game/Accessories/FORT_Backpacks/M_MED_Nautilus_Backpack/Meshes/Nautilus_Rig_Backpack_Nautilus_Male_MESH_Skeleton
graphs [
 AnimGraph {nodes [
   n1 = RigidBody!(ComponentPose=Local To Component(), Alpha, ExternalForce, ComponentLinearAccScale:Vector="0,0,0") -> (Pose) {
    bOverrideWorldGravity=true
    overridePhysicsAsset=/Game/Accessories/FORT_Backpacks/M_MED_Nautilus_Backpack/Meshes/Nautilus_Rig_Physics
    simulationSpace=BaseBoneSpace
    alphaBoolBlend {blendOption=Linear}
    baseBoneRef {boneName=root}
    overrideWorldGravity {z=-980}
   }
   Component To Local(ComponentPose=@n1)
   n2 = Apply Additive(
    Base=Local Space Ref Pose()
    Additive=Nautilus_Backpack_PoseAsset : piston01!(PoseWeight:float=1) -> (Pose:Pose Link Structure) {
     is=title AnimGraphNode_PoseByName
     poseAsset=&/Nautilus_Backpack_PoseAsset
     poseName=piston01
    }
    Alpha=Get piston01() -> (piston01:double)
   ) {alphaBoolBlend {blendOption=Linear}, alphaScaleBiasClamp {bInterpResult=true, interpSpeedDecreasing=30, interpSpeedIncreasing=30}}
   n3 = Apply Additive(
    Base=@n2
    Additive=Nautilus_Backpack_PoseAsset : piston02!(PoseWeight:float=1) -> (Pose:Pose Link Structure) {
     is=title AnimGraphNode_PoseByName
     poseAsset=&Nautilus_Backpack_PoseAsset
     poseName=piston02
    }
    Alpha=Get piston02() -> (piston02:double)
   ) {alphaBoolBlend {blendOption=Linear}, alphaScaleBiasClamp {bInterpResult=true, interpSpeedDecreasing=30, interpSpeedIncreasing=30}}
   n4 = Apply Additive(
    Base=@n3
    Additive=Nautilus_Backpack_PoseAsset : piston03!(PoseWeight:float=1) -> (Pose:Pose Link Structure) {
     is=title AnimGraphNode_PoseByName
     poseAsset=&Nautilus_Backpack_PoseAsset
     poseName=piston03
    }
    Alpha=Get piston03() -> (piston03:double)
   ) {alphaBoolBlend {blendOption=Linear}, alphaScaleBiasClamp {bInterpResult=true, interpSpeedDecreasing=30, interpSpeedIncreasing=30}}
   n5 = Apply Additive(
    Base=@n4
    Additive=Nautilus_Backpack_PoseAsset : piston04!(PoseWeight:float=1) -> (Pose:Pose Link Structure) {
     is=title AnimGraphNode_PoseByName
     poseAsset=&Nautilus_Backpack_PoseAsset
     poseName=piston04
    }
    Alpha=Get piston04() -> (piston04:double)
   ) {alphaBoolBlend {blendOption=Linear}, alphaScaleBiasClamp {bInterpResult=true, interpSpeedDecreasing=30, interpSpeedIncreasing=30}}
   Output Pose(Result=@n5) {layerGroup=None}
  ]}
 EventGraph {nodes [
   Event BlueprintUpdateAnimation()
   n1 = float * float(A=GetGameTimeInSeconds(), B=300)
   > Set piston01(piston01:double=MapRangeClamped(Value="Sin (Degrees)"!(A:double=@n1) -> (ReturnValue:double), InRangeA=-1, InRangeB=1)) -> (Output_Get:double)
   n2 = float + float(A=@n1, B=-50) {is=title K2Node_CommutativeAssociativeBinaryOperator}
   > Set piston02(piston02:double=MapRangeClamped(Value="Sin (Degrees)"(A=@n2), InRangeA=-1, InRangeB=1)) -> (Output_Get:double)
   n3 = float + float(A=@n1, B=-100)
   > Set piston03(piston03:double=MapRangeClamped(Value="Sin (Degrees)"(A=@n3), InRangeA=-1, InRangeB=1)) -> (Output_Get:double)
   n4 = float + float(A=@n1, B=-150)
   > Set piston04(piston04:double=MapRangeClamped(Value="Sin (Degrees)"(A=@n4), InRangeA=-1, InRangeB=1)) -> (Output_Get:double)
   TryGetPawnOwner()
  ]}
]
variables [
 piston01 {type=number}
 piston02 {type=number}
 piston03 {type=number}
 piston04 {type=number}
]
